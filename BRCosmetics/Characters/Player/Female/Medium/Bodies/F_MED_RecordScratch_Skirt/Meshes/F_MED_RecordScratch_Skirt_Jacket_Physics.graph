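# Physics asset for the jacket on the F_MED_RecordScratch_Skirt female player cosmetic, covering the back, front left/right and side left/right jacket bones.
# Each region pairs a small capsule on a base bone with a longer capsule on a dyn bone, offset along its X axis.
# Five constraints join each dyn bone to its base, allowing only limited Swing2 motion with twist and Swing1 limits at zero, so each panel swings like a narrow hinge.
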PhysicsAsset /BRCosmetics/Characters/Player/Female/Medium/Bodies/F_MED_RecordScratch_Skirt/Meshes/F_MED_RecordScratch_Skirt_Jacket_Physics
bodies {
 base_jacket_bk_a [{shapeType=Capsule, length=0.505, radius=0.505, shapeName=base_jacket_bk_a_capsule, rotation=<rot 0 -90 0>}]
 base_jacket_fr_a_l [{shapeType=Capsule, length=0.505, radius=0.505, shapeName=base_jacket_fr_a_l_capsule, rotation=<rot 0 -90 0>}]
 base_jacket_fr_a_r [{shapeType=Capsule, length=0.505, radius=0.505, shapeName=base_jacket_fr_a_r_capsule, rotation=<rot 0 -90 0>}]
 base_jacket_side_a_l [{shapeType=Capsule, length=0.505, radius=0.505, shapeName=base_jacket_side_a_l_capsule, rotation=<rot 0 -90 0>}]
 base_jacket_side_a_r [{shapeType=Capsule, length=0.505, radius=0.505, shapeName=base_jacket_side_a_r_capsule, rotation=<rot 0 -90 0>}]
 dyn_jacket_bk_a [{shapeType=Capsule, length=3, radius=1, shapeName=dyn_jacket_bk_a_capsule, center=<2 0 0>, rotation=<rot 90 0 0>}]
 dyn_jacket_fr_a_l [{shapeType=Capsule, length=3, radius=1, shapeName=dyn_jacket_fr_a_l_capsule, center=<2 0 0>, rotation=<rot 90 0 0>}]
 dyn_jacket_fr_a_r [{shapeType=Capsule, length=3, radius=1, shapeName=dyn_jacket_fr_a_r_capsule, center=<2 0 0>, rotation=<rot 90 0 0>}]
 dyn_jacket_side_a_l [{shapeType=Capsule, length=3, radius=1, shapeName=dyn_jacket_side_a_l_capsule, center=<2 0 0>, rotation=<rot 90 0 0>}]
 dyn_jacket_side_a_r [{shapeType=Capsule, length=3, radius=1, shapeName=dyn_jacket_side_a_r_capsule, center=<2 0 0>, rotation=<rot 90 0 0>}]
}
constraints [
 {bone1Name=dyn_jacket_bk_a, bone2Name=base_jacket_bk_a, swing1LimitDegrees=0, swing2LimitDegrees=10, twistLimitDegrees=0}
 {bone1Name=dyn_jacket_fr_a_l, bone2Name=base_jacket_fr_a_l, swing1LimitDegrees=0, swing2LimitDegrees=10, twistLimitDegrees=0}
 {bone1Name=dyn_jacket_fr_a_r, bone2Name=base_jacket_fr_a_r, swing1LimitDegrees=0, swing2LimitDegrees=10, twistLimitDegrees=0}
 {bone1Name=dyn_jacket_side_a_l, bone2Name=base_jacket_side_a_l, swing1LimitDegrees=0, swing2LimitDegrees=10, twistLimitDegrees=0}
 {bone1Name=dyn_jacket_side_a_r, bone2Name=base_jacket_side_a_r, swing1LimitDegrees=0, swing2LimitDegrees=10, twistLimitDegrees=0}
]
