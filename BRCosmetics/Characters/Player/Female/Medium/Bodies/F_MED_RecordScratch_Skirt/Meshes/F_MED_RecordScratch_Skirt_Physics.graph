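# Physics asset for the skirt on the RecordScratch female medium player body, providing capsule bodies and joints for skirt motion.
# Six two-segment dyn chains (front, side and back, left and right) hang from tiny base capsules that join through a middle base to the pelvis.
# Dyn joints have tight angular limits, tightest at the front with zero twist, while base and leg joints rotate freely.
# Pelvis, thigh and foot capsules supply the body shapes.
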PhysicsAsset /BRCosmetics/Characters/Player/Female/Medium/Bodies/F_MED_RecordScratch_Skirt/Meshes/F_MED_RecordScratch_Skirt_Physics
bodies {
 base_skirt_bk_a_l [{shapeType=Capsule, length=0.5, radius=0.5, shapeName=base_skirt_bk_a_l_capsule, rotation=<rot 90 -90 0>}]
 base_skirt_bk_a_r [{shapeType=Capsule, length=0.5, radius=0.5, shapeName=base_skirt_bk_a_r_capsule, rotation=<rot 90 -90 0>}]
 base_skirt_fr_a_l [{shapeType=Capsule, length=0.5, radius=0.5, shapeName=base_skirt_fr_a_l_capsule, rotation=<rot 90 -90 0>}]
 base_skirt_fr_a_r [{shapeType=Capsule, length=0.5, radius=0.5, shapeName=base_skirt_fr_a_r_capsule, rotation=<rot 90 -90 0>}]
 base_skirt_m_a [{shapeType=Capsule, length=0.5, radius=0.5, shapeName=base_skirt_m_a_capsule, rotation=<rot 90 -90 0>}]
 base_skirt_side_a_l [{shapeType=Capsule, length=0.5, radius=0.5, shapeName=base_skirt_side_a_l_capsule, rotation=<rot 90 -90 0>}]
 base_skirt_side_a_r [{shapeType=Capsule, length=0.5, radius=0.5, shapeName=base_skirt_side_a_r_capsule, rotation=<rot 90 -90 0>}]
 dyn_skirt_bk_a_l [{shapeType=Capsule, length=2.5, radius=5, shapeName=dyn_skirt_bk_a_l_capsule, center=<7 0 2.5>, rotation=<rot 90 0 0>}]
 dyn_skirt_bk_a_r [{shapeType=Capsule, length=2.5, radius=5, shapeName=dyn_skirt_bk_a_r_capsule, center=<7 0 -2.5>, rotation=<rot 90 0 0>}]
 dyn_skirt_bk_b_l [{shapeType=Capsule, length=5.25, radius=7, shapeName=dyn_skirt_bk_b_l_capsule, center=<7 0 2.5>, rotation=<rot 90 0 0>}]
 dyn_skirt_bk_b_r [{shapeType=Capsule, length=5.25, radius=7, shapeName=dyn_skirt_bk_b_r_capsule, center=<7 0 -2.5>, rotation=<rot 90 0 0>}]
 dyn_skirt_fr_a_l [{shapeType=Capsule, length=2.5, radius=5, shapeName=dyn_skirt_fr_a_l_capsule, center=<7 0 4>, rotation=<rot 90 0 0>}]
 dyn_skirt_fr_a_r [{shapeType=Capsule, length=2.5, radius=5, shapeName=dyn_skirt_fr_a_r_capsule, center=<7 0 -3.5>, rotation=<rot 90 0 0>}]
 dyn_skirt_fr_b_l [{shapeType=Capsule, length=5.25, radius=7, shapeName=dyn_skirt_fr_b_l_capsule, center=<7 0 2.5>, rotation=<rot 90 0 0>}]
 dyn_skirt_fr_b_r [{shapeType=Capsule, length=5.25, radius=7, shapeName=dyn_skirt_fr_b_r_capsule, center=<7 0 -1.5>, rotation=<rot 90 0 0>}]
 dyn_skirt_side_a_l [{shapeType=Capsule, length=2.5, radius=5, shapeName=dyn_skirt_side_a_l_capsule, center=<7 0 3.5>, rotation=<rot 90 0 0>}]
 dyn_skirt_side_a_r [{shapeType=Capsule, length=2.5, radius=5, shapeName=dyn_skirt_side_a_r_capsule, center=<7 0 -2>, rotation=<rot 90 0 0>}]
 dyn_skirt_side_b_l [{shapeType=Capsule, length=5.25, radius=7, shapeName=dyn_skirt_side_b_l_capsule, center=<7 0 2>, rotation=<rot 90 0 0>}]
 dyn_skirt_side_b_r [{shapeType=Capsule, length=5.25, radius=7, shapeName=dyn_skirt_side_b_r_capsule, center=<7 0 -2>, rotation=<rot 90 0 0>}]
 foot_l [{shapeType=Capsule, length=14, radius=7.5, shapeName=foot_l_capsule, center=<-4.27 7 0>, rotation=<rot 0 0 90>}]
 foot_r [{shapeType=Capsule, length=14, radius=7.5, shapeName=foot_r_capsule, center=<4.27 -7 0>, rotation=<rot 0 0 90>}]
 pelvis [{shapeType=Capsule, length=9.5, radius=9.5, center=<-1.7 0 0>}]
 thigh_l [{shapeType=Capsule, length=27.431, radius=7.5, shapeName=thigh_l_capsule, center=<-17 0 0.62>, rotation=<rot 90 0 0>}]
 thigh_r [{shapeType=Capsule, length=27.571, radius=7.36, shapeName=thigh_r_capsule, center=<17 0 -0.28>, rotation=<rot 90 0 0>}]
}
constraints [
 {
  bone1Name=base_skirt_side_a_l
  bone2Name=base_skirt_m_a
  swing1LimitDegrees=45
  swing1Motion=Free
  swing2LimitDegrees=45
  swing2Motion=Free
  twistLimitDegrees=45
  twistMotion=Free
 }
 {
  bone1Name=base_skirt_bk_a_l
  bone2Name=base_skirt_m_a
  swing1LimitDegrees=45
  swing1Motion=Free
  swing2LimitDegrees=45
  swing2Motion=Free
  twistLimitDegrees=45
  twistMotion=Free
 }
 {
  bone1Name=base_skirt_bk_a_r
  bone2Name=base_skirt_m_a
  swing1LimitDegrees=45
  swing1Motion=Free
  swing2LimitDegrees=45
  swing2Motion=Free
  twistLimitDegrees=45
  twistMotion=Free
 }
 {
  bone1Name=base_skirt_fr_a_l
  bone2Name=base_skirt_m_a
  swing1LimitDegrees=45
  swing1Motion=Free
  swing2LimitDegrees=45
  swing2Motion=Free
  twistLimitDegrees=45
  twistMotion=Free
 }
 {
  bone1Name=base_skirt_fr_a_r
  bone2Name=base_skirt_m_a
  swing1LimitDegrees=45
  swing1Motion=Free
  swing2LimitDegrees=45
  swing2Motion=Free
  twistLimitDegrees=45
  twistMotion=Free
 }
 {
  bone1Name=base_skirt_side_a_r
  bone2Name=base_skirt_m_a
  swing1LimitDegrees=45
  swing1Motion=Free
  swing2LimitDegrees=45
  swing2Motion=Free
  twistLimitDegrees=45
  twistMotion=Free
 }
 {
  bone1Name=base_skirt_m_a
  bone2Name=pelvis
  swing1LimitDegrees=45
  swing1Motion=Free
  swing2LimitDegrees=45
  swing2Motion=Free
  twistLimitDegrees=45
  twistMotion=Free
 }
 {bone1Name=dyn_skirt_bk_a_l, bone2Name=base_skirt_bk_a_l, swing1LimitDegrees=10, swing2LimitDegrees=20, twistLimitDegrees=5}
 {bone1Name=dyn_skirt_bk_b_l, bone2Name=dyn_skirt_bk_a_l, swing1LimitDegrees=10, swing2LimitDegrees=20, twistLimitDegrees=5}
 {bone1Name=dyn_skirt_bk_a_r, bone2Name=base_skirt_bk_a_r, swing1LimitDegrees=10, swing2LimitDegrees=20, twistLimitDegrees=5}
 {bone1Name=dyn_skirt_bk_b_r, bone2Name=dyn_skirt_bk_a_r, swing1LimitDegrees=10, swing2LimitDegrees=20, twistLimitDegrees=5}
 {bone1Name=dyn_skirt_fr_a_l, bone2Name=base_skirt_fr_a_l, swing1LimitDegrees=5, swing2LimitDegrees=15, twistLimitDegrees=0}
 {bone1Name=dyn_skirt_fr_b_l, bone2Name=dyn_skirt_fr_a_l, swing1LimitDegrees=5, swing2LimitDegrees=7.5, twistLimitDegrees=0}
 {bone1Name=dyn_skirt_fr_a_r, bone2Name=base_skirt_fr_a_r, swing1LimitDegrees=5, swing2LimitDegrees=15, twistLimitDegrees=0}
 {bone1Name=dyn_skirt_fr_b_r, bone2Name=dyn_skirt_fr_a_r, swing1LimitDegrees=5, swing2LimitDegrees=7.5, twistLimitDegrees=0}
 {bone1Name=dyn_skirt_side_a_l, bone2Name=base_skirt_side_a_l, swing1LimitDegrees=5, swing2LimitDegrees=20, twistLimitDegrees=5}
 {bone1Name=dyn_skirt_side_b_l, bone2Name=dyn_skirt_side_a_l, swing1LimitDegrees=5, swing2LimitDegrees=20, twistLimitDegrees=5}
 {bone1Name=dyn_skirt_side_a_r, bone2Name=base_skirt_side_a_r, swing1LimitDegrees=5, swing2LimitDegrees=20, twistLimitDegrees=5}
 {bone1Name=dyn_skirt_side_b_r, bone2Name=dyn_skirt_side_a_r, swing1LimitDegrees=5, swing2LimitDegrees=20, twistLimitDegrees=5}
 {
  bone1Name=thigh_r
  bone2Name=pelvis
  swing1LimitDegrees=45
  swing1Motion=Free
  swing2LimitDegrees=45
  swing2Motion=Free
  twistLimitDegrees=45
  twistMotion=Free
 }
 {
  bone1Name=thigh_l
  bone2Name=pelvis
  swing1LimitDegrees=45
  swing1Motion=Free
  swing2LimitDegrees=45
  swing2Motion=Free
  twistLimitDegrees=45
  twistMotion=Free
 }
 {
  bone1Name=foot_l
  bone2Name=thigh_l
  swing1LimitDegrees=45
  swing1Motion=Free
  swing2LimitDegrees=45
  swing2Motion=Free
  twistLimitDegrees=45
  twistMotion=Free
 }
 {
  bone1Name=foot_r
  bone2Name=thigh_r
  swing1LimitDegrees=45
  swing1Motion=Free
  swing2LimitDegrees=45
  swing2Motion=Free
  twistLimitDegrees=45
  twistMotion=Free
 }
]
