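# Physics asset for the necktie on the female medium RecordScratch Skirt character body, covering the tie bones and upper spine.
# A four-bone dyn_tie_fr chain hangs off base_tie_m_a, with small swing limits and zero twist limit at each joint.
# Capsules form the collision shapes, doubled side by side on the two end tie bones and much larger on spine_04 and spine_05.
# Constraints from tie base to spine_05 and between the spine bones leave rotation free.
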PhysicsAsset /BRCosmetics/Characters/Player/Female/Medium/Bodies/F_MED_RecordScratch_Skirt/Meshes/F_MED_RecordScratch_Skirt_Tie_Physics
bodies {
 base_tie_m_a [{shapeType=Capsule, length=0.505, radius=0.505, shapeName=base_tie_m_a_capsule, rotation=<rot 0 -90 0>}]
 dyn_tie_fr_a [{shapeType=Capsule, length=1, radius=1, shapeName=dyn_tie_fr_a_capsule, center=<0.5 0 0>, rotation=<rot 90 0 0>}]
 dyn_tie_fr_b [{shapeType=Capsule, length=3.2139, radius=1.1752, shapeName=dyn_tie_fr_b_capsule, center=<1.8565 0 0.8318>, rotation=<rot 90 0 0>}]
 dyn_tie_fr_c [
  {shapeType=Capsule, length=3.2139, radius=1.1752, shapeName=dyn_tie_fr_b_capsule, center=<1.7839 1.253 0.7468>, rotation=<rot 90 0 -11.9042>}
  {shapeType=Capsule, length=3.2139, radius=1.1752, shapeName=dyn_tie_fr_b_capsule, center=<1.748 -1.3289 0.7468>, rotation=<rot 90 0 11.9042>}
 ]
 dyn_tie_fr_d [
  {
   shapeType=Capsule
   length=3.2139
   radius=1.1752
   shapeName=dyn_tie_fr_b_capsule
   center=<2.6209 1.1769 0.1733>
   rotation=<rot 77.4187 -28.6944 -0.0001>
  }
  {
   shapeType=Capsule
   length=3.4164
   radius=1.1752
   shapeName=dyn_tie_fr_b_capsule
   center=<2.5953 -1.1248 -0.028>
   rotation=<rot 75.8298 26.9231 -3.5264>
  }
 ]
 spine_04 [{shapeType=Capsule, length=8.3582, radius=8.5756, shapeName=spine_05_capsule, center=<2.0828 5.7101 0>}]
 spine_05 [{
   shapeType=Capsule
   length=4.6179
   radius=11.2617
   shapeName=spine_05_capsule
   center=<-2.1234 0.1905 0>
   rotation=<rot 90 176.4237 -143.1882>
  }]
}
constraints [
 {bone1Name=dyn_tie_fr_a, bone2Name=base_tie_m_a, swing1LimitDegrees=10, swing2LimitDegrees=8, twistLimitDegrees=0}
 {bone1Name=dyn_tie_fr_b, bone2Name=dyn_tie_fr_a, swing1LimitDegrees=5, swing2LimitDegrees=5, twistLimitDegrees=0}
 {bone1Name=dyn_tie_fr_c, bone2Name=dyn_tie_fr_b, swing1LimitDegrees=5, swing2LimitDegrees=5, twistLimitDegrees=0}
 {bone1Name=dyn_tie_fr_d, bone2Name=dyn_tie_fr_c, swing1LimitDegrees=5, swing2LimitDegrees=5, twistLimitDegrees=0}
 {
  bone1Name=base_tie_m_a
  bone2Name=spine_05
  swing1LimitDegrees=45
  swing1Motion=Free
  swing2LimitDegrees=45
  swing2Motion=Free
  twistLimitDegrees=45
  twistMotion=Free
 }
 {
  bone1Name=spine_05
  bone2Name=spine_04
  swing1LimitDegrees=45
  swing1Motion=Free
  swing2LimitDegrees=45
  swing2Motion=Free
  twistLimitDegrees=45
  twistMotion=Free
 }
]
