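# Post-process style AnimBlueprint for the female RecordScratch skirt body, keeping skirt, jacket, tie, belt and cape in shape as the player moves.
# A copied parent pose runs through a Control Rig, pelvis/thigh skirt constraints, then cached stages of state-weighted additive bone corrections, including thigh-rotation-mapped crouch fixes.
# Three RigidBody simulations then add secondary motion, followed by final bone fixes.
# The EventGraph caches pawn and mesh references and tracks states like ziplining.
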AnimBlueprint /BRCosmetics/Characters/Player/Female/Medium/Bodies/F_MED_RecordScratch_Skirt/Meshes/F_MED_RecordScratch_Skirt_AnimBP
target_skeleton=&/F_MED_RecordScratch_Skirt_Skeleton
graphs [
 AnimGraph {nodes [
   Output Pose(Result=Use cached pose 'CRCache_CapeAdj'()) {layerGroup=None}
   n1 = Control Rig(Source=Copy Pose From Mesh() {bUseAttachedParent=true}) {
    alphaBoolBlend {blendOption=Linear}
    controlRigAssetReference {blueprintRigClass=&/F_MED_DeformRig_RecordScratch_Skirt_ControlRig_new.F_MED_DeformRig_RecordScratch_Skirt_ControlRig_new_C}
   }
   Save cached pose 'CRCache_ControlRig'(Pose=@n1) {cacheName=CRCache_ControlRig}
   n2 = Get gravityOverride() -> (gravityOverride:Vector)
   n3 = RigidBody!(
    ComponentPose=Local To Component(LocalPose=Use cached pose 'CRCache_BeltFix'())
    Alpha=Get SkirtDynAlpha() -> (SkirtDynAlpha:double)
    OverrideWorldGravity:Vector=@n2
    ExternalForce
   ) -> (Pose) {
    bClampLinearTranslationLimitToRefPose=true
    bOverrideWorldGravity=true
    overridePhysicsAsset=&/F_MED_RecordScratch_Skirt_Physics
    simulationSpace=BaseBoneSpace
    alphaBoolBlend {blendOption=Linear}
    alphaScaleBiasClamp {bInterpResult=true, outRange {max=0.5, min=1}}
    baseBoneRef {boneName=base_skirt_m_a}
    componentAppliedLinearAccClamp=<3000 3000 3000>
    componentLinearAccScale=<0.25 0.25 0.25>
    componentLinearVelScale=<0.25 0.25 0.25>
    overrideWorldGravity {z=-980}
   }
   n4 = RigidBody(ComponentPose=@n3, Alpha=Get skydiveAlpha() -> (skydiveAlpha:double), OverrideWorldGravity=@n2) {
    bOverrideWorldGravity=true
    overridePhysicsAsset=&/F_MED_RecordScratch_Skirt_Tie_Physics
    simulationSpace=BaseBoneSpace
    alphaBoolBlend {blendOption=Linear}
    baseBoneRef {boneName=spine_04}
    componentAppliedLinearAccClamp=<3000 3000 3000>
    componentLinearAccScale=<0.3 0.3 0.3>
    componentLinearVelScale=<0.3 0.3 0.3>
   }
   n5 = RigidBody!(
    ComponentPose=@n4
    Alpha
    OverrideWorldGravity=@n2
    ExternalForce
    SimSpaceSettings:Sim Space Settings Structure=Get Sim Space Settings SkyDive() -> (Sim Space Settings SkyDive:Sim Space Settings Structure)
   ) -> (Pose) {
    bOverrideWorldGravity=true
    overridePhysicsAsset=&/F_MED_RecordScratch_Skirt_Jacket_Physics
    simulationSpace=BaseBoneSpace
    alphaBoolBlend {blendOption=Linear}
    baseBoneRef {boneName=pelvis}
    componentAppliedLinearAccClamp=<3000 3000 3000>
    componentLinearAccScale=<0.3 0.3 0.4>
    componentLinearVelScale=<0.3 0.3 0.4>
    simSpaceSettings {
     maxAngularAcceleration=1000
     maxAngularVelocity=3
     maxLinearAcceleration=700
     maxLinearVelocity=500
     velocityScaleZ=0
     worldAlpha=0.3
     externalLinearDragV=<0.2 0.2>
    }
   }
   n6 = "Constraint - base_skirt_fr_a (PP)"!(ComponentPose=@n5, Alpha=1, ConstraintWeights_0=0.5, ConstraintWeights_1:float=0.5) -> (Pose) {
    alphaBoolBlend {blendOption=Linear}
    boneToModify {boneName=base_skirt_fr_a}
    constraintSetup [{targetBone {boneName=dyn_skirt_fr_a_l}}, {targetBone {boneName=dyn_skirt_fr_a_r}}]
    constraintWeights [1,1]
   }
   n7 = "Constraint - dyn_skirt_bk_a (PP)"!(ComponentPose=@n6, Alpha=1, ConstraintWeights_0=0.5, ConstraintWeights_1:float=0.5) -> (Pose) {
    alphaBoolBlend {blendOption=Linear}
    boneToModify {boneName=dyn_skirt_bk_a}
    constraintSetup [{targetBone {boneName=dyn_skirt_bk_a_l}}, {targetBone {boneName=dyn_skirt_bk_a_r}}]
    constraintWeights [1,1]
   }
   n8 = "Transform (Modify) Bone - Bone: base_skirt_fr_a"(ComponentPose=@n7, Alpha=Get FrSkirtAlpha() -> (FrSkirtAlpha:double), Translation="0,0,1.5", Rotation="10,0,-10", Scale="1,1,1") {
    rotationMode=BMM_Additive
    rotationSpace=BCS_BoneSpace
    scaleMode=BMM_Additive
    translationMode=BMM_Additive
    translationSpace=BCS_BoneSpace
    alphaBoolBlend {blendOption=Linear}
    alphaScaleBiasClamp {bInterpResult=true}
    boneToModify {boneName=base_skirt_fr_a}
    rotation {pitch=10, roll=-10}
    translation {z=1.5}
   }
   n9 = Get isCrouchMove() -> (isCrouchMove:double)
   n10 = "Transform (Modify) Bone - Bone: dyn_skirt_fr_b_r"(ComponentPose=@n8, Alpha=@n9, Translation="0,0,0", Rotation="0,15,0", Scale="1,1,1") {
    rotationMode=BMM_Additive
    rotationSpace=BCS_BoneSpace
    scaleMode=BMM_Additive
    translationMode=BMM_Additive
    translationSpace=BCS_BoneSpace
    alphaBoolBlend {blendOption=Linear}
    alphaScaleBiasClamp {bInterpResult=true}
    boneToModify {boneName=dyn_skirt_fr_b_r}
    rotation {yaw=15}
   }
   n11 = "Transform (Modify) Bone - Bone: dyn_skirt_fr_b"(ComponentPose=@n10, Alpha=@n9, Translation="0,0,1", Rotation="0,0,0", Scale="1,1,1") {
    rotationMode=BMM_Additive
    rotationSpace=BCS_BoneSpace
    scaleMode=BMM_Additive
    translationMode=BMM_Additive
    translationSpace=BCS_BoneSpace
    alphaBoolBlend {blendOption=Linear}
    alphaScaleBiasClamp {bInterpResult=true}
    boneToModify {boneName=dyn_skirt_fr_b}
    translation {z=1}
   }
   n12 = "Transform (Modify) Bone - Bone: base_tie_m_a"(ComponentPose=@n11, Alpha=Get FrontEndLevel() -> (FrontEndLevel:double), Translation="0,0,0", Rotation="-3,0,0", Scale="1,1,1") {
    rotationMode=BMM_Additive
    rotationSpace=BCS_BoneSpace
    scaleMode=BMM_Additive
    translationSpace=BCS_BoneSpace
    alphaBoolBlend {blendOption=Linear}
    alphaScaleBiasClamp {bInterpResult=true}
    boneToModify {boneName=base_tie_m_a}
    rotation {pitch=-3}
   }
   Save cached pose 'CRCache_Dyn'(Pose=Component To Local(ComponentPose=@n12)) {cacheName=CRCache_Dyn}
   n13 = Get EmoteDance() -> (EmoteDance:double)
   n14 = "Transform (Modify) Bone - Bone: base_skirt_fr_a"(
    ComponentPose=Local To Component(LocalPose=Use cached pose 'L_CrouchFix'())
    Alpha=@n13
    Translation="1.5,0,6"
    Rotation="90,0,0"
    Scale="1,1,1"
   ) {
    rotationMode=BMM_Additive
    rotationSpace=BCS_BoneSpace
    scaleMode=BMM_Additive
    translationMode=BMM_Additive
    translationSpace=BCS_BoneSpace
    alphaBoolBlend {blendOption=Linear}
    alphaScaleBiasClamp {bInterpResult=true}
    boneToModify {boneName=base_skirt_fr_a}
    rotation {pitch=90}
    translation {x=1.5, z=6}
   }
   n15 = "Transform (Modify) Bone - Bone: base_skirt_fr_a_l"(ComponentPose=@n14, Alpha=@n13, Translation="1.5,1.5,1.5", Rotation="0,0,0", Scale="1,1,1") {
    rotationMode=BMM_Additive
    rotationSpace=BCS_BoneSpace
    scaleMode=BMM_Additive
    translationMode=BMM_Additive
    translationSpace=BCS_BoneSpace
    alphaBoolBlend {blendOption=Linear}
    alphaScaleBiasClamp {bInterpResult=true}
    boneToModify {boneName=base_skirt_fr_a_l}
    translation=<1.5 1.5 1.5>
   }
   n16 = "Transform (Modify) Bone - Bone: base_skirt_fr_a_r"(ComponentPose=@n15, Alpha=@n13, Translation="1.5,1.5,-1.5", Rotation="0,0,0", Scale="1,1,1") {
    rotationMode=BMM_Additive
    rotationSpace=BCS_BoneSpace
    scaleMode=BMM_Additive
    translationMode=BMM_Additive
    translationSpace=BCS_BoneSpace
    alphaBoolBlend {blendOption=Linear}
    alphaScaleBiasClamp {bInterpResult=true}
    boneToModify {boneName=base_skirt_fr_a_r}
    translation=<1.5 1.5 -1.5>
   }
   Save cached pose 'CRCache_EmoteSkirt'(Pose=Component To Local(ComponentPose=@n16)) {cacheName=CRCache_EmoteSkirt}
   n17 = "Constraint - base_skirt_fr_a_r (RR)"!(
    ComponentPose=Local To Component(LocalPose=Use cached pose 'CRCache_ControlRig'())
    Alpha=1
    ConstraintWeights_0=0.1
    ConstraintWeights_1:float=0.9
   ) -> (Pose) {
    alphaBoolBlend {blendOption=Linear}
    boneToModify {boneName=base_skirt_fr_a_r}
    constraintSetup [{transformType=Rotation, targetBone {boneName=pelvis}}, {transformType=Rotation, targetBone {boneName=thigh_r}}]
    constraintWeights [1,1]
   }
   n18 = "Constraint - base_skirt_side_a_r (RR)"!(ComponentPose=@n17, Alpha=1, ConstraintWeights_0=0.15, ConstraintWeights_1:float=0.85) -> (Pose) {
    alphaBoolBlend {blendOption=Linear}
    boneToModify {boneName=base_skirt_side_a_r}
    constraintSetup [{transformType=Rotation, targetBone {boneName=pelvis}}, {transformType=Rotation, targetBone {boneName=thigh_r}}]
    constraintWeights [1,1]
   }
   n19 = "Constraint - base_skirt_bk_a_r (RR)"!(ComponentPose=@n18, Alpha=1, ConstraintWeights_0=0.9, ConstraintWeights_1:float=0.1) -> (Pose) {
    alphaBoolBlend {blendOption=Linear}
    boneToModify {boneName=base_skirt_bk_a_r}
    constraintSetup [{transformType=Rotation, targetBone {boneName=pelvis}}, {transformType=Rotation, targetBone {boneName=thigh_r}}]
    constraintWeights [1,1]
   }
   n20 = "Constraint - base_skirt_fr_a_l (RR)"!(ComponentPose=@n19, Alpha=1, ConstraintWeights_0=0.1, ConstraintWeights_1:float=0.9) -> (Pose) {
    alphaBoolBlend {blendOption=Linear}
    boneToModify {boneName=base_skirt_fr_a_l}
    constraintSetup [{transformType=Rotation, targetBone {boneName=pelvis}}, {transformType=Rotation, targetBone {boneName=thigh_l}}]
    constraintWeights [1,1]
   }
   n21 = "Constraint - base_skirt_side_a_l (RR)"!(ComponentPose=@n20, Alpha=1, ConstraintWeights_0=0.4, ConstraintWeights_1:float=0.6) -> (Pose) {
    alphaBoolBlend {blendOption=Linear}
    boneToModify {boneName=base_skirt_side_a_l}
    constraintSetup [{transformType=Rotation, targetBone {boneName=pelvis}}, {transformType=Rotation, targetBone {boneName=thigh_l}}]
    constraintWeights [1,1]
   }
   n22 = "Constraint - base_skirt_bk_a_l (RR)"!(ComponentPose=@n21, Alpha=1, ConstraintWeights_0=0.9, ConstraintWeights_1:float=0.1) -> (Pose) {
    alphaBoolBlend {blendOption=Linear}
    boneToModify {boneName=base_skirt_bk_a_l}
    constraintSetup [{transformType=Rotation, targetBone {boneName=pelvis}}, {transformType=Rotation, targetBone {boneName=thigh_l}}]
    constraintWeights [1,1]
   }
   Save cached pose 'CRCache_SkirtC'(Pose=Component To Local(ComponentPose=@n22)) {cacheName=CRCache_SkirtC}
   n23 = Get CrouchingAlpha() -> (CrouchingAlpha:double)
   n24 = Get isCrouchMove()
   n25 = "Transform (Modify) Bone - Bone: base_skirt_side_a_r"(
    ComponentPose=Local To Component(LocalPose=Use cached pose 'CRCache_EmoteSkirt'())
    Alpha=@n24
    Translation="0,0,0"
    Rotation="0,15,0"
    Scale="1,1,1"
   ) {
    rotationMode=BMM_Additive
    rotationSpace=BCS_BoneSpace
    scaleMode=BMM_Additive
    scaleSpace=BCS_BoneSpace
    translationMode=BMM_Additive
    translationSpace=BCS_BoneSpace
    alphaBoolBlend {blendOption=Linear}
    alphaScaleBiasClamp {bInterpResult=true}
    boneToModify {boneName=base_skirt_side_a_r}
    rotation {yaw=15}
   }
   n26 = "Transform (Modify) Bone - Bone: base_skirt_fr_a_r"(ComponentPose=@n25, Alpha=@n24, Translation="2,0,0", Rotation="25,0,0", Scale="1,1,1") {
    rotationMode=BMM_Additive
    rotationSpace=BCS_BoneSpace
    scaleMode=BMM_Additive
    scaleSpace=BCS_BoneSpace
    translationMode=BMM_Additive
    translationSpace=BCS_BoneSpace
    alphaBoolBlend {blendOption=Linear}
    alphaScaleBiasClamp {bInterpResult=true}
    boneToModify {boneName=base_skirt_fr_a_r}
    rotation {pitch=25}
    translation {x=2}
   }
   n27 = "Transform (Modify) Bone - Bone: thigh_twist_01_r"(ComponentPose=@n26, Alpha=@n24, Translation="0,2,0", Rotation="0,0,0", Scale="1,0.8,1") {
    rotationMode=BMM_Additive
    rotationSpace=BCS_BoneSpace
    scaleMode=BMM_Additive
    scaleSpace=BCS_BoneSpace
    translationMode=BMM_Additive
    translationSpace=BCS_BoneSpace
    alphaBoolBlend {blendOption=Linear}
    alphaScaleBiasClamp {bInterpResult=true, interpSpeedDecreasing=5, interpSpeedIncreasing=5}
    boneToModify {boneName=thigh_twist_01_r}
    scale {y=0.8}
    translation {y=2}
   }
   n28 = Get DBNOAlpha() -> (DBNOAlpha:double)
   n29 = "Transform (Modify) Bone - Bone: base_skirt_fr_a_r"(ComponentPose=@n27, Alpha=@n28, Translation="6,0,-7", Rotation="-5,-20,0", Scale="1,1,1") {
    rotationMode=BMM_Additive
    rotationSpace=BCS_BoneSpace
    scaleMode=BMM_Additive
    translationMode=BMM_Additive
    translationSpace=BCS_BoneSpace
    alphaBoolBlend {blendOption=Linear}
    alphaScaleBiasClamp {bInterpResult=true}
    boneToModify {boneName=base_skirt_fr_a_r}
    rotation {pitch=-5, yaw=-20}
    translation {x=6, z=-7}
   }
   n30 = "Transform (Modify) Bone - Bone: base_skirt_fr_a_l"(ComponentPose=@n29, Alpha=@n28, Translation="6.022,0,7", Rotation="2,-20,0", Scale="1,1,1") {
    rotationMode=BMM_Additive
    rotationSpace=BCS_BoneSpace
    scaleMode=BMM_Additive
    translationMode=BMM_Additive
    translationSpace=BCS_BoneSpace
    alphaBoolBlend {blendOption=Linear}
    alphaScaleBiasClamp {bInterpResult=true}
    boneToModify {boneName=base_skirt_fr_a_l}
    rotation {pitch=2, yaw=-20}
    translation {x=6.022, z=7}
   }
   n31 = Get bkSkirtAlpha() -> (bkSkirtAlpha:double)
   n32 = "Transform (Modify) Bone - Bone: base_skirt_bk_a_r"(ComponentPose=@n30, Alpha=@n31, Translation="0,0,-2.5", Rotation="15,0,0", Scale="1,1,1") {
    rotationMode=BMM_Additive
    rotationSpace=BCS_BoneSpace
    scaleMode=BMM_Additive
    translationMode=BMM_Additive
    translationSpace=BCS_BoneSpace
    alphaBoolBlend {blendOption=Linear}
    alphaScaleBiasClamp {bInterpResult=true}
    boneToModify {boneName=base_skirt_bk_a_r}
    rotation {pitch=15}
    translation {z=-2.5}
   }
   n33 = "Transform (Modify) Bone - Bone: base_skirt_bk_a_l"(ComponentPose=@n32, Alpha=@n31, Translation="0,0,2.5", Rotation="-15,0,0", Scale="1,1,1") {
    rotationMode=BMM_Additive
    rotationSpace=BCS_BoneSpace
    scaleMode=BMM_Additive
    translationMode=BMM_Additive
    translationSpace=BCS_BoneSpace
    alphaBoolBlend {blendOption=Linear}
    alphaScaleBiasClamp {bInterpResult=true}
    boneToModify {boneName=base_skirt_bk_a_l}
    rotation {pitch=-15}
    translation {z=2.5}
   }
   n34 = Get isKneeSlide() -> (isKneeSlide:double)
   n35 = "Transform (Modify) Bone - Bone: base_skirt_bk_a_r"(ComponentPose=@n33, Alpha=@n34, Translation="0,0,8", Rotation="-10.000,0.000,110.000", Scale="1,1,1") {
    rotationMode=BMM_Additive
    scaleMode=BMM_Additive
    translationMode=BMM_Additive
    alphaBoolBlend {blendOption=Linear}
    alphaScaleBiasClamp {bInterpResult=true, interpSpeedDecreasing=5, interpSpeedIncreasing=5}
    boneToModify {boneName=base_skirt_bk_a_r}
    rotation=<rot -10 0 110>
    translation {z=8}
   }
   n36 = "Transform (Modify) Bone - Bone: base_skirt_bk_a_l"(ComponentPose=@n35, Alpha=@n34, Translation="0,0,8", Rotation="0,0,100", Scale="1,1,1") {
    rotationMode=BMM_Additive
    scaleMode=BMM_Additive
    translationMode=BMM_Additive
    alphaBoolBlend {blendOption=Linear}
    alphaScaleBiasClamp {bInterpResult=true, interpSpeedDecreasing=5, interpSpeedIncreasing=5}
    boneToModify {boneName=base_skirt_bk_a_l}
    rotation {roll=100}
    translation {z=8}
   }
   n37 = "Transform (Modify) Bone - Bone: base_skirt_side_a_r"(ComponentPose=@n36, Alpha=@n34, Translation="0,0,5", Rotation="-33.263,-9.954,58.755", Scale="1,1,1") {
    rotationMode=BMM_Additive
    scaleMode=BMM_Additive
    translationMode=BMM_Additive
    alphaBoolBlend {blendOption=Linear}
    alphaScaleBiasClamp {bInterpResult=true, interpSpeedDecreasing=5, interpSpeedIncreasing=5}
    boneToModify {boneName=base_skirt_side_a_r}
    rotation=<rot -33.263 -9.954 58.755>
    translation {z=5}
   }
   n38 = "Transform (Modify) Bone - Bone: base_skirt_side_a_l"(ComponentPose=@n37, Alpha=@n34, Translation="0,0,5", Rotation="33.263,9.954,58.755", Scale="1,1,1") {
    rotationMode=BMM_Additive
    scaleMode=BMM_Additive
    translationMode=BMM_Additive
    alphaBoolBlend {blendOption=Linear}
    alphaScaleBiasClamp {bInterpResult=true, interpSpeedDecreasing=5, interpSpeedIncreasing=5}
    boneToModify {boneName=base_skirt_side_a_l}
    rotation=<rot 33.263 9.954 58.755>
    translation {z=5}
   }
   n39 = "Transform (Modify) Bone - Bone: base_skirt_fr_a_r"(ComponentPose=@n38, Alpha=Get isWalking() -> (isWalking:double), Translation="1,-2,-2.5", Rotation="-5,-5,0", Scale="1,1,1") {
    rotationMode=BMM_Additive
    rotationSpace=BCS_BoneSpace
    scaleMode=BMM_Additive
    translationMode=BMM_Additive
    translationSpace=BCS_BoneSpace
    alphaBoolBlend {blendOption=Linear}
    alphaScaleBiasClamp {bInterpResult=true}
    boneToModify {boneName=base_skirt_fr_a_r}
    rotation {pitch=-5, yaw=-5}
    translation=<1 -2 -2.5>
   }
   n40 = "Transform (Modify) Bone - Bone: base_skirt_bk_a_r"(ComponentPose=@n39, Alpha=1, Translation="0,0,-1.5", Rotation="0,0,0", Scale="1,1,1") {
    rotationSpace=BCS_BoneSpace
    scaleMode=BMM_Additive
    translationMode=BMM_Additive
    translationSpace=BCS_BoneSpace
    alphaBoolBlend {blendOption=Linear}
    boneToModify {boneName=base_skirt_bk_a_r}
    translation {z=-1.5}
   }
   n41 = "Transform (Modify) Bone - Bone: base_skirt_bk_a_l"(ComponentPose=@n40, Alpha=1, Translation="0,0,1.5", Rotation="0,0,0", Scale="1,1,1") {
    rotationSpace=BCS_BoneSpace
    scaleMode=BMM_Additive
    translationMode=BMM_Additive
    translationSpace=BCS_BoneSpace
    alphaBoolBlend {blendOption=Linear}
    boneToModify {boneName=base_skirt_bk_a_l}
    translation {z=1.5}
   }
   n42 = Get OnGrindRail() -> (OnGrindRail:double)
   n43 = "Transform (Modify) Bone - Bone: base_skirt_side_a_l"(ComponentPose=@n41, Alpha=@n42, Translation="0,0,0", Rotation="2.539,30.578,-28.390", Scale="1,1,1") {
    rotationMode=BMM_Additive
    rotationSpace=BCS_BoneSpace
    translationMode=BMM_Additive
    translationSpace=BCS_BoneSpace
    alphaBoolBlend {blendOption=Linear}
    boneToModify {boneName=base_skirt_side_a_l}
    rotation=<rot 2.539 30.5781 -28.39>
   }
   n44 = "Transform (Modify) Bone - Bone: base_skirt_side_a_r"(ComponentPose=@n43, Alpha=@n42, Translation="9.568,5.514,-0.997", Rotation="-3.746,38.978,-37.850", Scale="1,1,1") {
    rotationMode=BMM_Additive
    rotationSpace=BCS_BoneSpace
    translationMode=BMM_Additive
    translationSpace=BCS_BoneSpace
    alphaBoolBlend {blendOption=Linear}
    boneToModify {boneName=base_skirt_side_a_r}
    rotation=<rot -3.746 38.978 -37.8499>
    translation=<9.568 5.514 -0.997>
   }
   n45 = "Transform (Modify) Bone - Bone: base_skirt_fr_a_r"(ComponentPose=@n44, Alpha=@n23, Translation="0,0,0", Rotation="0,0,0", Scale="1,1,1") {
    rotationMode=BMM_Additive
    rotationSpace=BCS_BoneSpace
    scaleMode=BMM_Additive
    scaleSpace=BCS_BoneSpace
    translationMode=BMM_Additive
    translationSpace=BCS_BoneSpace
    alphaBoolBlend {blendOption=Linear}
    alphaScaleBiasClamp {bInterpResult=true}
    boneToModify {boneName=base_skirt_fr_a_r}
   }
   n46 = "Transform (Modify) Bone - Bone: base_skirt_fr_a_l"(ComponentPose=@n45, Alpha=@n23, Translation="1,0,1", Rotation="0,0,25", Scale="1,1,1") {
    rotationMode=BMM_Additive
    rotationSpace=BCS_BoneSpace
    scaleMode=BMM_Additive
    scaleSpace=BCS_BoneSpace
    translationMode=BMM_Additive
    translationSpace=BCS_BoneSpace
    alphaBoolBlend {blendOption=Linear}
    alphaScaleBiasClamp {bInterpResult=true}
    boneToModify {boneName=base_skirt_fr_a_l}
    rotation {roll=25}
    translation {x=1, z=1}
   }
   n47 = "Transform (Modify) Bone - Bone: base_skirt_side_a_r"(ComponentPose=@n46, Alpha=@n23, Translation="0,0,0", Rotation="0,0,15", Scale="1,1,1") {
    rotationMode=BMM_Additive
    rotationSpace=BCS_BoneSpace
    scaleMode=BMM_Additive
    scaleSpace=BCS_BoneSpace
    translationMode=BMM_Additive
    translationSpace=BCS_BoneSpace
    alphaBoolBlend {blendOption=Linear}
    alphaScaleBiasClamp {bInterpResult=true}
    boneToModify {boneName=base_skirt_side_a_r}
    rotation {roll=15}
   }
   n48 = "Transform (Modify) Bone - Bone: base_skirt_side_a_r"(ComponentPose=@n47, Alpha=Get JumpAlpha() -> (JumpAlpha:double), Translation="0,0,0", Rotation="-10,25,7", Scale="1,1,1") {
    rotationMode=BMM_Additive
    rotationSpace=BCS_BoneSpace
    scaleMode=BMM_Additive
    translationMode=BMM_Additive
    translationSpace=BCS_BoneSpace
    alphaBoolBlend {blendOption=Linear}
    alphaScaleBiasClamp {bInterpResult=true, interpSpeedDecreasing=5, interpSpeedIncreasing=5}
    boneToModify {boneName=base_skirt_side_a_r}
    rotation=<rot -10 25 7>
   }
   n49 = "Transform (Modify) Bone - Bone: base_skirt_fr_a_r"(ComponentPose=@n48, Alpha=Get JumpMoveAlpha() -> (JumpMoveAlpha:double), Translation="0,0,0", Rotation="0,0,-35", Scale="1,1,1") {
    rotationMode=BMM_Additive
    rotationSpace=BCS_BoneSpace
    scaleMode=BMM_Additive
    translationMode=BMM_Additive
    translationSpace=BCS_BoneSpace
    alphaBoolBlend {blendOption=Linear}
    alphaScaleBiasClamp {bInterpResult=true, interpSpeedDecreasing=5, interpSpeedIncreasing=5}
    boneToModify {boneName=base_skirt_fr_a_r}
    rotation {roll=-35}
   }
   n50 = Get AlphaHandRelaxed2() -> (AlphaHandRelaxed2:float)
   n51 = "Transform (Modify) Bone - Bone: base_skirt_side_a_r"(ComponentPose=@n49, Alpha=@n50, Translation="0,0,0", Rotation="15,0.01,0", Scale="1,1,1") {
    rotationMode=BMM_Additive
    rotationSpace=BCS_BoneSpace
    scaleMode=BMM_Additive
    translationSpace=BCS_BoneSpace
    alphaBoolBlend {blendOption=Linear}
    alphaScaleBiasClamp {bInterpResult=true}
    boneToModify {boneName=base_skirt_side_a_r}
    rotation {pitch=15, yaw=0.01}
   }
   n52 = "Transform (Modify) Bone - Bone: base_skirt_side_a_l"(ComponentPose=@n51, Alpha=@n50, Translation="0,0,0", Rotation="-15,0,0", Scale="1,1,1") {
    rotationMode=BMM_Additive
    rotationSpace=BCS_BoneSpace
    scaleMode=BMM_Additive
    translationSpace=BCS_BoneSpace
    alphaBoolBlend {blendOption=Linear}
    alphaScaleBiasClamp {bInterpResult=true}
    boneToModify {boneName=base_skirt_side_a_l}
    rotation {pitch=-15}
   }
   n53 = Get isParachuting() -> (isParachuting:double)
   n54 = "Transform (Modify) Bone - Bone: base_skirt_side_a_r"(ComponentPose=@n52, Alpha=@n53, Translation="0,0,-1.5", Rotation="-10,0,0", Scale="1,1,1") {
    rotationMode=BMM_Additive
    rotationSpace=BCS_BoneSpace
    scaleMode=BMM_Additive
    translationMode=BMM_Additive
    translationSpace=BCS_BoneSpace
    alphaBoolBlend {blendOption=Linear}
    alphaScaleBiasClamp {bInterpResult=true}
    boneToModify {boneName=base_skirt_side_a_r}
    rotation {pitch=-10}
    translation {z=-1.5}
   }
   n55 = "Transform (Modify) Bone - Bone: base_skirt_fr_a_r"(ComponentPose=@n54, Alpha=@n53, Translation="0,0,0", Rotation="-15,0,0", Scale="1,1,1") {
    rotationMode=BMM_Additive
    rotationSpace=BCS_BoneSpace
    scaleMode=BMM_Additive
    translationMode=BMM_Additive
    translationSpace=BCS_BoneSpace
    alphaBoolBlend {blendOption=Linear}
    alphaScaleBiasClamp {bInterpResult=true}
    boneToModify {boneName=base_skirt_fr_a_r}
    rotation {pitch=-15}
   }
   n56 = "Transform (Modify) Bone - Bone: base_skirt_side_a_l"(ComponentPose=@n55, Alpha=@n53, Translation="0,2,2", Rotation="10,0,0", Scale="1,1,1") {
    rotationMode=BMM_Additive
    rotationSpace=BCS_BoneSpace
    scaleMode=BMM_Additive
    translationMode=BMM_Additive
    translationSpace=BCS_BoneSpace
    alphaBoolBlend {blendOption=Linear}
    alphaScaleBiasClamp {bInterpResult=true}
    boneToModify {boneName=base_skirt_side_a_l}
    rotation {pitch=10}
    translation {y=2, z=2}
   }
   n57 = "Transform (Modify) Bone - Bone: base_skirt_fr_a_l"(ComponentPose=@n56, Alpha=@n53, Translation="0,0,0", Rotation="15,0,0", Scale="1,1,1") {
    rotationMode=BMM_Additive
    rotationSpace=BCS_BoneSpace
    scaleMode=BMM_Additive
    translationMode=BMM_Additive
    translationSpace=BCS_BoneSpace
    alphaBoolBlend {blendOption=Linear}
    alphaScaleBiasClamp {bInterpResult=true}
    boneToModify {boneName=base_skirt_fr_a_l}
    rotation {pitch=15}
   }
   n58 = Get isinAir() -> (isinAir:double)
   n59 = "Transform (Modify) Bone - Bone: base_skirt_bk_a_r"(ComponentPose=@n57, Alpha=@n58, Translation="0,2.5,-2.5", Rotation="10,0,0", Scale="1,1,1") {
    rotationMode=BMM_Additive
    rotationSpace=BCS_BoneSpace
    scaleMode=BMM_Additive
    translationMode=BMM_Additive
    translationSpace=BCS_BoneSpace
    alphaBoolBlend {blendOption=Linear}
    alphaScaleBiasClamp {bInterpResult=true, interpSpeedDecreasing=2, interpSpeedIncreasing=2}
    boneToModify {boneName=base_skirt_bk_a_r}
    rotation {pitch=10}
    translation {y=2.5, z=-2.5}
   }
   n60 = "Transform (Modify) Bone - Bone: base_skirt_bk_a_l"(ComponentPose=@n59, Alpha=@n58, Translation="0,2.5,2.5", Rotation="-10,0,0", Scale="1,1,1") {
    rotationMode=BMM_Additive
    rotationSpace=BCS_BoneSpace
    scaleMode=BMM_Additive
    translationMode=BMM_Additive
    translationSpace=BCS_BoneSpace
    alphaBoolBlend {blendOption=Linear}
    alphaScaleBiasClamp {bInterpResult=true, interpSpeedDecreasing=2, interpSpeedIncreasing=2}
    boneToModify {boneName=base_skirt_bk_a_l}
    rotation {pitch=-10}
    translation {y=2.5, z=2.5}
   }
   n61 = "Transform (Modify) Bone - Bone: base_skirt_side_a_l"(ComponentPose=@n60, Alpha=@n58, Translation="0,0,0", Rotation="10,0,0", Scale="1,1,1") {
    rotationMode=BMM_Additive
    rotationSpace=BCS_BoneSpace
    scaleMode=BMM_Additive
    translationMode=BMM_Additive
    translationSpace=BCS_BoneSpace
    alphaBoolBlend {blendOption=Linear}
    alphaScaleBiasClamp {bInterpResult=true}
    boneToModify {boneName=base_skirt_side_a_l}
    rotation {pitch=10}
   }
   n62 = "Transform (Modify) Bone - Bone: base_skirt_fr_a_r"(ComponentPose=@n61, Alpha=Get Standing() -> (Standing:double), Translation="0,0,0", Rotation="44.572,-0.160,-41.806", Scale="1,1,1") {
    rotationMode=BMM_Additive
    rotationSpace=BCS_BoneSpace
    translationMode=BMM_Additive
    translationSpace=BCS_BoneSpace
    alphaBoolBlend {blendOption=Linear}
    alphaScaleBiasClamp {bInterpResult=true}
    boneToModify {boneName=base_skirt_fr_a_r}
    rotation=<rot 44.572 -0.1599 -41.8059>
   }
   n63 = "Transform (Modify) Bone - Bone: base_skirt_side_a_r"!(
    ComponentPose=@n62
    bAlphaBoolEnabled:bool=Get OnZipline() -> (OnZipline:bool)
    Translation="0,0,0"
    Rotation="-21.420,-6.265,-10.655"
    Scale="1,1,1"
   ) -> (Pose) {
    alphaInputType=Bool
    rotationMode=BMM_Additive
    rotationSpace=BCS_BoneSpace
    translationMode=BMM_Additive
    translationSpace=BCS_BoneSpace
    alphaBoolBlend {blendOption=Linear}
    alphaScaleBiasClamp {bInterpResult=true}
    boneToModify {boneName=base_skirt_side_a_r}
    rotation=<rot -21.42 -6.2649 -10.655>
   }
   n64 = "Transform (Modify) Bone - Bone: base_skirt_fr_a_l"!(
    ComponentPose=@n63
    bAlphaBoolEnabled:bool=Get bIsInVehicleSeat() -> (bIsInVehicleSeat:bool)
    Translation="0,0,0"
    Rotation="-0.009,-0.001,80.000"
    Scale="1,1,1"
   ) -> (Pose) {
    alphaInputType=Bool
    rotationMode=BMM_Additive
    rotationSpace=BCS_BoneSpace
    translationMode=BMM_Additive
    translationSpace=BCS_BoneSpace
    alphaBoolBlend {blendOption=Linear}
    alphaScaleBiasClamp {bInterpResult=true}
    boneToModify {boneName=base_skirt_fr_a_l}
    rotation=<rot -0.009 -0.001 80>
   }
   Save cached pose 'CRCache_SkirtAdj'(Pose=Component To Local(ComponentPose=@n64)) {cacheName=CRCache_SkirtAdj}
   n65 = "Transform (Modify) Bone - Bone: attach_cape"(
    ComponentPose=Local To Component(LocalPose=Use cached pose 'CRCache_Dyn'())
    Alpha=Get CapeFixer() -> (CapeFixer:float)
    Translation="0,-4,0"
    Rotation="0,0,0"
    Scale="1,1,1"
   ) {
    rotationMode=BMM_Additive
    rotationSpace=BCS_BoneSpace
    scaleMode=BMM_Additive
    scaleSpace=BCS_BoneSpace
    translationMode=BMM_Additive
    translationSpace=BCS_BoneSpace
    alphaBoolBlend {blendOption=Linear}
    alphaScaleBiasClamp {bInterpResult=true}
    boneToModify {boneName=attach_cape}
    translation {y=-4}
   }
   Save cached pose 'CRCache_CapeAdj'(Pose=Component To Local(ComponentPose=@n65)) {cacheName=CRCache_CapeAdj}
   n66 = "Transform (Modify) Bone - Bone: base_tie_m_a"(
    ComponentPose=Local To Component(LocalPose=Use cached pose 'CRCache_JacketAdj'())
    Alpha=Get isInSkydive() -> (isInSkydive:double)
    Translation="0,0,0"
    Rotation="4,0,0"
    Scale="1,1,1"
   ) {
    rotationMode=BMM_Additive
    rotationSpace=BCS_BoneSpace
    scaleMode=BMM_Additive
    translationSpace=BCS_BoneSpace
    alphaBoolBlend {blendOption=Linear}
    alphaScaleBiasClamp {bInterpResult=true}
    boneToModify {boneName=base_tie_m_a}
    rotation {pitch=4}
   }
   Save cached pose 'CRCache_TieAdj'(Pose=Component To Local(ComponentPose=@n66)) {cacheName=CRCache_TieAdj}
   n67 = Get thigh_rot_r()
   n68 = "Transform (Modify) Bone - Bone: base_skirt_fr_a_r"(
    ComponentPose=Local To Component(LocalPose=Use cached pose 'CRCache_SkirtC'())
    Alpha=@n67
    Translation="-5,7.5,-6"
    Rotation="-5,15,10"
    Scale="1,1,1"
   ) {
    rotationMode=BMM_Additive
    rotationSpace=BCS_BoneSpace
    scaleMode=BMM_Additive
    scaleSpace=BCS_BoneSpace
    translationMode=BMM_Additive
    translationSpace=BCS_BoneSpace
    alphaBoolBlend {blendOption=Linear}
    alphaScaleBiasClamp {bMapRange=true, inRange {max=0, min=-100}, outRange {max=0, min=1}}
    boneToModify {boneName=base_skirt_fr_a_r}
    rotation=<rot -5 15 10>
    translation=<-5 7.5 -6>
   }
   n69 = Component To Local(ComponentPose=@n68)
   n70 = "Transform (Modify) Bone - Bone: base_skirt_fr_a_r"(
    ComponentPose=Local To Component(LocalPose=Use cached pose 'CRCache_SkirtC'())
    Alpha=Get thigh_rot_r()
    Translation="-8,12.272,-10"
    Rotation="-20,0,0"
    Scale="1,1,1"
   ) {
    rotationMode=BMM_Additive
    rotationSpace=BCS_BoneSpace
    scaleMode=BMM_Additive
    scaleSpace=BCS_BoneSpace
    translationMode=BMM_Additive
    translationSpace=BCS_BoneSpace
    alphaBoolBlend {blendOption=Linear}
    alphaScaleBiasClamp {bMapRange=true, inRange {max=0, min=-100}, outRange {max=0, min=1}}
    boneToModify {boneName=base_skirt_fr_a_r}
    rotation {pitch=-20}
    translation=<-8 12.272 -10>
   }
   n71 = Component To Local(ComponentPose=@n70)
   n72 = Get CrouchLevel_R() -> (CrouchLevel_R:bool)
   n73 = "Transform (Modify) Bone - Bone: base_skirt_side_a_r"!(
    ComponentPose=Local To Component(LocalPose=Blend Poses by bool(BlendPose_0=@n71, BlendPose_1=@n69, bActiveValue=@n72) {blendType=Linear})
    Alpha:float=@n67
    Translation="0,0,-4"
    Rotation="-10,-5,0"
    Scale="1,1,1"
   ) -> (Pose) {
    rotationMode=BMM_Additive
    rotationSpace=BCS_BoneSpace
    scaleMode=BMM_Additive
    scaleSpace=BCS_BoneSpace
    translationMode=BMM_Additive
    translationSpace=BCS_BoneSpace
    alphaBoolBlend {blendOption=Linear}
    alphaScaleBiasClamp {bMapRange=true, inRange {max=0, min=-100}, outRange {max=0, min=1}}
    boneToModify {boneName=base_skirt_side_a_r}
    rotation {pitch=-10, yaw=-5}
    translation {z=-4}
   }
   n74 = "Transform (Modify) Bone - Bone: base_skirt_bk_a_r"(ComponentPose=@n73, Alpha=@n67, Translation="0,0,-2", Rotation="0,0,0", Scale="1,1,1") {
    rotationMode=BMM_Additive
    rotationSpace=BCS_BoneSpace
    scaleMode=BMM_Additive
    scaleSpace=BCS_BoneSpace
    translationMode=BMM_Additive
    translationSpace=BCS_BoneSpace
    alphaBoolBlend {blendOption=Linear}
    alphaScaleBiasClamp {bMapRange=true, inRange {max=0, min=-100}, outRange {max=0, min=1}}
    boneToModify {boneName=base_skirt_bk_a_r}
    translation {z=-2}
   }
   Save cached pose 'R_CrouchFix'(Pose=Component To Local(ComponentPose=@n74)) {cacheName=R_CrouchFix}
   n75 = "Transform (Modify) Bone - Bone: base_skirt_fr_a_l"!(
    ComponentPose=Local To Component(LocalPose=Use cached pose 'R_CrouchFix'())
    Alpha:float=Get thigh_rot_l()
    Translation="2,10.5,16.5"
    Rotation="5.1,0,20"
    Scale="1,1,1"
   ) -> (Pose) {
    rotationMode=BMM_Additive
    rotationSpace=BCS_BoneSpace
    scaleMode=BMM_Additive
    scaleSpace=BCS_BoneSpace
    translationMode=BMM_Additive
    translationSpace=BCS_BoneSpace
    alphaBoolBlend {blendOption=Linear}
    alphaScaleBiasClamp {bMapRange=true, inRange {max=0, min=-100}, outRange {max=0, min=1}}
    boneToModify {boneName=base_skirt_fr_a_l}
    rotation {pitch=5.1, roll=20}
    translation=<2 10.5 16.5>
   }
   n76 = Component To Local(ComponentPose=@n75)
   n77 = Get thigh_rot_l()
   n78 = "Transform (Modify) Bone - Bone: base_skirt_fr_a_l"(
    ComponentPose=Local To Component(LocalPose=Use cached pose 'R_CrouchFix'())
    Alpha=@n77
    Translation="-5,7.5,12.5"
    Rotation="0.832,10.051,14.754"
    Scale="1,1,1"
   ) {
    rotationMode=BMM_Additive
    rotationSpace=BCS_BoneSpace
    scaleMode=BMM_Additive
    scaleSpace=BCS_BoneSpace
    translationMode=BMM_Additive
    translationSpace=BCS_BoneSpace
    alphaBoolBlend {blendOption=Linear}
    alphaScaleBiasClamp {bMapRange=true, inRange {max=0, min=-100}, outRange {max=0, min=1}}
    boneToModify {boneName=base_skirt_fr_a_l}
    rotation=<rot 0.832 10.051 14.754>
    translation=<-5 7.5 12.5>
   }
   n79 = Component To Local(ComponentPose=@n78)
   n80 = Get CrouchLevel_L() -> (CrouchLevel_L:bool)
   n81 = "Transform (Modify) Bone - Bone: base_skirt_side_a_l"(
    ComponentPose=Local To Component(LocalPose=Blend Poses by bool(BlendPose_0=@n76, BlendPose_1=@n79, bActiveValue=@n80) {blendType=Linear})
    Alpha=@n77
    Translation="0,0,4"
    Rotation="0,0,0"
    Scale="1,1,1"
   ) {
    rotationMode=BMM_Additive
    rotationSpace=BCS_BoneSpace
    scaleMode=BMM_Additive
    scaleSpace=BCS_BoneSpace
    translationMode=BMM_Additive
    translationSpace=BCS_BoneSpace
    alphaBoolBlend {blendOption=Linear}
    alphaScaleBiasClamp {bMapRange=true, inRange {max=0, min=-100}, outRange {max=0, min=1}}
    boneToModify {boneName=base_skirt_side_a_l}
    translation {z=4}
   }
   n82 = "Transform (Modify) Bone - Bone: base_skirt_bk_a_l"(ComponentPose=@n81, Alpha=@n77, Translation="0,0,2", Rotation="0,0,0", Scale="1,1,1") {
    rotationMode=BMM_Additive
    rotationSpace=BCS_BoneSpace
    scaleMode=BMM_Additive
    scaleSpace=BCS_BoneSpace
    translationMode=BMM_Additive
    translationSpace=BCS_BoneSpace
    alphaBoolBlend {blendOption=Linear}
    alphaScaleBiasClamp {bMapRange=true, inRange {max=0, min=-100}, outRange {max=0, min=1}}
    boneToModify {boneName=base_skirt_bk_a_l}
    translation {z=2}
   }
   Save cached pose 'L_CrouchFix'(Pose=Component To Local(ComponentPose=@n82)) {cacheName=L_CrouchFix}
   n83 = "Transform (Modify) Bone - Bone: base_belt_fr_a"(
    ComponentPose=Local To Component(LocalPose=Use cached pose 'CRCache_TieAdj'())
    Alpha=Get isCrouchMove()
    Translation="-5.15,0,0"
    Rotation="0,0,0"
    Scale="1,1,1"
   ) {
    rotationMode=BMM_Additive
    rotationSpace=BCS_BoneSpace
    scaleMode=BMM_Additive
    translationMode=BMM_Additive
    translationSpace=BCS_BoneSpace
    alphaBoolBlend {blendOption=Linear}
    alphaScaleBiasClamp {bInterpResult=true}
    boneToModify {boneName=base_belt_fr_a}
    translation {x=-5.15}
   }
   n84 = "Transform (Modify) Bone - Bone: base_belt_fr_a"(ComponentPose=@n83, Alpha=Get JumpAlpha(), Translation="0,0,0", Rotation="0,10,0", Scale="1,1,1") {
    rotationMode=BMM_Additive
    rotationSpace=BCS_BoneSpace
    scaleMode=BMM_Additive
    translationMode=BMM_Additive
    translationSpace=BCS_BoneSpace
    alphaBoolBlend {blendOption=Linear}
    alphaScaleBiasClamp {bInterpResult=true, interpSpeedDecreasing=5}
    boneToModify {boneName=base_belt_fr_a}
    rotation {yaw=10}
   }
   n85 = "Transform (Modify) Bone - Bone: base_belt_fr_a"(ComponentPose=@n84, Alpha=Get EmoteDance(), Translation="-2.5,0,0", Rotation="0,0,0", Scale="1,1,1") {
    rotationMode=BMM_Additive
    rotationSpace=BCS_BoneSpace
    scaleMode=BMM_Additive
    translationMode=BMM_Additive
    translationSpace=BCS_BoneSpace
    alphaBoolBlend {blendOption=Linear}
    alphaScaleBiasClamp {bInterpResult=true, interpSpeedDecreasing=5, interpSpeedIncreasing=5}
    boneToModify {boneName=base_belt_fr_a}
    translation {x=-2.5}
   }
   n86 = "Transform (Modify) Bone - Bone: base_belt_fr_a"(ComponentPose=@n85, Alpha=Get isKneeSlide(), Translation="5.15,0,0", Rotation="0,10,0", Scale="1,1,1") {
    rotationMode=BMM_Additive
    rotationSpace=BCS_BoneSpace
    scaleMode=BMM_Additive
    translationMode=BMM_Additive
    translationSpace=BCS_BoneSpace
    alphaBoolBlend {blendOption=Linear}
    alphaScaleBiasClamp {bInterpResult=true}
    boneToModify {boneName=base_belt_fr_a}
    rotation {yaw=10}
    translation {x=5.15}
   }
   n87 = "Transform (Modify) Bone - Bone: base_belt_fr_a"(ComponentPose=@n86, Alpha=Get isWalking(), Translation="-2.5,0,0", Rotation="0,0,0", Scale="1,1,1") {
    rotationMode=BMM_Additive
    rotationSpace=BCS_BoneSpace
    scaleMode=BMM_Additive
    translationMode=BMM_Additive
    translationSpace=BCS_BoneSpace
    alphaBoolBlend {blendOption=Linear}
    alphaScaleBiasClamp {bInterpResult=true, interpSpeedDecreasing=5, interpSpeedIncreasing=5}
    boneToModify {boneName=base_belt_fr_a}
    translation {x=-2.5}
   }
   Save cached pose 'CRCache_BeltFix'(Pose=Component To Local(ComponentPose=@n87)) {cacheName=CRCache_BeltFix}
   n88 = Get isCrouchMove()
   n89 = "Transform (Modify) Bone - Bone: base_jacket_bk_a"(
    ComponentPose=Local To Component(LocalPose=Use cached pose 'CRCache_SkirtAdj'())
    Alpha=@n88
    Translation="1.5,1,0"
    Rotation="-20.000,0.000,-0.000"
    Scale="1,1,1"
   ) {
    rotationMode=BMM_Additive
    rotationSpace=BCS_BoneSpace
    scaleMode=BMM_Additive
    translationMode=BMM_Additive
    translationSpace=BCS_BoneSpace
    alphaBoolBlend {blendOption=Linear}
    alphaScaleBiasClamp {bInterpResult=true, interpSpeedIncreasing=5}
    boneToModify {boneName=base_jacket_bk_a}
    rotation=<rot -20 0 0>
    translation=<1.5 1>
   }
   n90 = "Transform (Modify) Bone - Bone: base_jacket_side_a_r"(ComponentPose=@n89, Alpha=@n88, Translation="1,1,0", Rotation="0,0,0", Scale="1,1,1") {
    rotationMode=BMM_Additive
    rotationSpace=BCS_BoneSpace
    scaleMode=BMM_Additive
    translationMode=BMM_Additive
    translationSpace=BCS_BoneSpace
    alphaBoolBlend {blendOption=Linear}
    alphaScaleBiasClamp {bInterpResult=true, interpSpeedIncreasing=5}
    boneToModify {boneName=base_jacket_side_a_r}
    translation=<1 1>
   }
   n91 = "Transform (Modify) Bone - Bone: base_jacket_side_a_l"(ComponentPose=@n90, Alpha=@n88, Translation="1,1,0", Rotation="0,0,0", Scale="1,1,1") {
    rotationMode=BMM_Additive
    rotationSpace=BCS_BoneSpace
    scaleMode=BMM_Additive
    translationMode=BMM_Additive
    translationSpace=BCS_BoneSpace
    alphaBoolBlend {blendOption=Linear}
    alphaScaleBiasClamp {bInterpResult=true, interpSpeedIncreasing=5}
    boneToModify {boneName=base_jacket_side_a_l}
    translation=<1 1>
   }
   n92 = "Transform (Modify) Bone - Bone: base_jacket_m_a"(ComponentPose=@n91, Alpha=Get Is Sprinting() -> (Is Sprinting:float), Translation="0,0,0", Rotation="10,0,0", Scale="1,1,1") {
    rotationMode=BMM_Additive
    rotationSpace=BCS_BoneSpace
    scaleMode=BMM_Additive
    translationSpace=BCS_BoneSpace
    alphaBoolBlend {blendOption=Linear}
    alphaScaleBiasClamp {bInterpResult=true, interpSpeedDecreasing=7, interpSpeedIncreasing=7}
    boneToModify {boneName=base_jacket_m_a}
    rotation {pitch=10}
   }
   Save cached pose 'CRCache_JacketAdj'(Pose=Component To Local(ComponentPose=@n92)) {cacheName=CRCache_JacketAdj}
  ]}
 EventGraph {nodes [
   Event BlueprintInitializeAnimation()
   > Set SkelMesh(SkelMesh=GetOwnerSkeletalMeshComponent())
   n1 = TryGetPawnOwner()
   > Is Valid(InputObject=@n1)
   > n2 = Cast To FortPlayerPawnAthena(Object=@n1)
   > n3 = Set PlayerPawn(PlayerPawn:Fort Player Pawn Athena Object Reference=@n2) -> (Output_Get:Fort Player Pawn Athena Object Reference)
   > Set Body_Mesh(Body_Mesh=GetSkeletalMeshForPartType(self=@n3))
   n4 = Event BlueprintProcessAnimInputEvent()
   > n5 = Branch(Condition=IsNamed(self=@n4.AnimInputEvent, InEventName=aie.Player.Zipline.Entered))
   > Set OnZipline(OnZipline:bool=true) -> (Output_Get:bool)
   Branch(execute=@n5.else, Condition=IsNamed(self=@n4.AnimInputEvent, InEventName=aie.Player.Zipline.Exited))
   > Set OnZipline(OnZipline=false)
   n6 = Event BlueprintUpdateAnimation()
   > n7 = Sequence!(execute) -> (then_0, then_1, then_2:Exec, then_3:Exec, then_4:Exec)
   n8 = Branch(
    execute=@n7.then_1
    Condition=OR Boolean!(A=Get bIsSkyDiving(), B=Get bIsSkydiveDiveMode(), C:bool=Get bIsParachuteOpened(), D:bool=Get bIsSkydiveFloating()) -> (ReturnValue)
   )
   > Set Sim Space Settings SkyDive(Sim Space Settings SkyDive:Sim Space Settings Structure=Get Sim Space Settings Skydiving() -> (Sim Space Settings Skydiving:Sim Space Settings Structure)) -> (Output_Get:Sim Space Settings Structure)
   Set Sim Space Settings SkyDive(execute=@n8.else, Sim Space Settings SkyDive=Get Sim Space Settings Regular() -> (Sim Space Settings Regular:Sim Space Settings Structure))
   n9 = Sequence!(execute=@n7.then_4) -> (then_0, then_1, then_2, then_3, then_4, then_5:Exec, then_6:Exec, then_7:Exec)
   n10 = Branch(execute=@n9.then_3, Condition=OR Boolean!(A=Get CrouchLevel_L(), B=Get CrouchLevel_R(), C=Get bIsKneeSliding()) -> (ReturnValue))
   > Set SkirtDynAlpha(SkirtDynAlpha:double=0) -> (Output_Get:double)
   Set SkirtDynAlpha(execute=@n10.else, SkirtDynAlpha=1)
   > n11 = Branch(Condition=Get bIsSkydiveFloating())
   > Set skydiveAlpha(skydiveAlpha:double=0.75) -> (Output_Get:double)
   Set skydiveAlpha(execute=@n11.else, skydiveAlpha=1)
   > n12 = Branch(Condition=Get bIsCrouching())
   > Set SkirtDynAlpha(SkirtDynAlpha=0.5)
   Set skydiveAlpha(execute=@n12.else, skydiveAlpha=1)
   n13 = OR Boolean!(
    A=OR Boolean!(A=Get bIsSkyDiving(), B=Get bIsSkydiveDiveMode(), C:bool=Get bIsParachuteOpened(), D:bool=Get bIsSkydiveFloating()) -> (ReturnValue)
    B=Get bIsSurfaceSwimming()
   ) -> (ReturnValue)
   n14 = Branch(execute=@n9.then_0, Condition=@n13)
   > Set skydiveAlpha(skydiveAlpha=0.65)
   Set skydiveAlpha(execute=@n14.else, skydiveAlpha=1)
   n15 = Branch(execute=@n9.then_5, Condition=OR Boolean!(A=Get bIsCrouching(), B=Get bIsCrouchSprinting()) -> (ReturnValue))
   > n16 = Branch(Condition=Get bIsKneeSliding())
   > Set bkSkirtAlpha(bkSkirtAlpha:double=0) -> (Output_Get:double)
   Set bkSkirtAlpha(execute=@n16.else, bkSkirtAlpha=1)
   Set bkSkirtAlpha(execute=@n15.else, bkSkirtAlpha=0)
   n17 = AND Boolean(A=Get bIsCrouching(), B=Get bIsMoving2D())
   n18 = Get bIsSprinting()
   Set JacketAlpha(execute=@n9.then_6, JacketAlpha:double="To Float (Boolean)"(InBool=OR Boolean(A=@n17, B=@n18))) -> (Output_Get:double)
   n19 = NOT Boolean(A=Get bIsCrouching())
   n20 = NOT Boolean(A=Get bIsTargeting())
   > Set Is Sprinting(Is Sprinting:float="To Float (Boolean)"(InBool=AND Boolean!(A=@n18, B=@n19, C:bool=@n20) -> (ReturnValue))) -> (Output_Get:float)
   n21 = Branch(execute=@n9.then_4, Condition=OR Boolean!(A=Get bIsCrouching(), B=Get bIsJumping(), C:bool=Get bIsFalling()) -> (ReturnValue))
   > Set FrSkirtAlpha(FrSkirtAlpha:double=0) -> (Output_Get:double)
   Set FrSkirtAlpha(execute=@n21.else, FrSkirtAlpha=1)
   Set isInSkydive(execute=@n9.then_1, isInSkydive:double="To Float (Boolean)"(InBool=@n13)) -> (Output_Get:double)
   Set FrontEndLevel(execute=@n9.then_7, FrontEndLevel:double="To Float (Boolean)"(InBool=Get bIsInFrontEnd())) -> (Output_Get:double)
   n22 = OR Boolean!(A=Get bIsJumping(), B=Get bIsFalling()) -> (ReturnValue)
   n23 = Get bIsSkydiveDiveMode()
   n24 = Get bIsSkydiveFloating()
   n25 = Get bIsParachuteOpened()
   n26 = Get bIsSkyDiving()
   Set isinAir(
    execute=@n9.then_2
    isinAir:double="To Float (Boolean)"(InBool=OR Boolean!(A=@n26, B=@n23, C:bool=@n24, D:bool=@n25, E:bool=@n22) -> (ReturnValue))
   ) -> (Output_Get:double)
   > Set isParachuting(isParachuting:double="To Float (Boolean)"(InBool=Get bIsParachuteOpened())) -> (Output_Get:double)
   > Set EmoteDance(EmoteDance:double="To Float (Boolean)"(InBool=Get bIsPlayingEmote())) -> (Output_Get:double)
   > Set OnGrindRail(OnGrindRail:double="To Float (Boolean)"(InBool=Get bIsOnGrindRail())) -> (Output_Get:double)
   n27 = NOT Boolean(A=Get bIsPlayingEmote())
   n28 = NOT Boolean(A=Get bIsMoving2D())
   n29 = NOT Boolean(A=Get bIsCrouching())
   n30 = Get bRigidBodyEnabled()
   > Set Standing(Standing:double="To Float (Boolean)"(InBool=AND Boolean!(A=@n28, B=@n29, C=@n27, D:bool=@n30) -> (ReturnValue))) -> (Output_Get:double)
   > Set isKneeSlide(isKneeSlide:double="To Float (Boolean)"(InBool=Get bIsKneeSliding())) -> (Output_Get:double)
   > Set DBNOAlpha(DBNOAlpha:double="To Float (Boolean)"(InBool=Get bIsDBNO())) -> (Output_Get:double)
   n31 = OR Boolean!(A=Get bIsJumping(), B=Get bIsFalling()) -> (ReturnValue)
   n32 = Get bIsMoving2D()
   > Set JumpMoveAlpha(JumpMoveAlpha:double="To Float (Boolean)"(InBool=AND Boolean!(A=@n31, B=@n32) -> (ReturnValue))) -> (Output_Get:double)
   n33 = Get bIsJumping()
   n34 = Get bIsFalling()
   > Set JumpAlpha(JumpAlpha:double="To Float (Boolean)"(InBool=OR Boolean(A=@n33, B=@n34))) -> (Output_Get:double)
   n35 = AND Boolean!(A=Get bIsSprinting(), B=NOT Boolean(A=Get bIsCrouching()), C:bool=NOT Boolean(A=Get bIsTargeting())) -> (ReturnValue)
   n36 = Get bIsMoving2D()
   > Set isWalking(isWalking:double="To Float (Boolean)"(InBool=OR Boolean(A=@n36, B=@n35))) -> (Output_Get:double)
   n37 = Get bIsMoving2D()
   n38 = Get bIsCrouching()
   > Set isCrouchMove(isCrouchMove:double="To Float (Boolean)"(InBool=AND Boolean!(A=@n37, B=@n38) -> (ReturnValue))) -> (Output_Get:double)
   n39 = NOT Boolean(A=@n37)
   > Set CrouchingAlpha(CrouchingAlpha:double="To Float (Boolean)"(InBool=AND Boolean(A=@n39, B=@n38))) -> (Output_Get:double)
   n40 = Get bIsSkydiveFloating()
   n41 = Get bIsSkydiveDiveMode()
   n42 = Get bIsSkyDiving()
   n43 = Get bIsParachuteOpened()
   n44 = Get bIsJumping()
   n45 = Get bIsFalling()
   > n46 = Branch(Condition=OR Boolean(A=OR Boolean!(A=@n42, B=@n41, C:bool=@n40, D:bool=@n43) -> (ReturnValue), B=OR Boolean!(A=@n44, B=@n45) -> (ReturnValue)))
   > Set CapeFixer(CapeFixer:float=1) -> (Output_Get:float)
   Set CapeFixer(execute=@n46.else, CapeFixer=0)
   n47 = GetDeltaTransformFromRefPose(self=GetOwnerSkeletalMeshComponent(), BoneName=thigh_l, BaseName=pelvis)
   Set thigh_rot_l(execute=@n7.then_2, thigh_rot_l=BreakRotator(InRot=BreakTransform(InTransform=@n47).Rotation).Yaw)
   n48 = GetDeltaTransformFromRefPose(self=GetOwnerSkeletalMeshComponent(), BoneName=thigh_r, BaseName=pelvis)
   > Set thigh_rot_r(thigh_rot_r=BreakRotator(InRot=BreakTransform(InTransform=@n48).Rotation).Yaw)
   > n49 = Sequence!(execute) -> (then_0, then_1)
   n50 = Get thigh_rot_r()
   > n51 = Branch(Condition=AND Boolean(A="float <= float"(A=@n50, B=-110) -> (ReturnValue=~), B=Get bIsCrouching()))
   > Set CrouchLevel_R(CrouchLevel_R:bool=true) -> (Output_Get:bool)
   Set CrouchLevel_R(execute=@n51.else, CrouchLevel_R=false)
   n52 = Get thigh_rot_l()
   n53 = Branch(execute=@n49.then_1, Condition=AND Boolean(A="float <= float"(A=@n52, B=-130) -> (ReturnValue=~), B=Get bIsCrouching()))
   > Set CrouchLevel_L(CrouchLevel_L:bool=true) -> (Output_Get:bool)
   Set CrouchLevel_L(execute=@n53.else, CrouchLevel_L=false)
   n54 = Cast To FortPlayerAnimInstance(execute=@n7.then_3, Object=GetAnimInstance(self=GetOwnerSkeletalMeshComponent()))
   > n55 = Switch on EFortWeaponCoreAnimation!(execute, Selection:EFortWeaponCoreAnimation Enum=Get WeaponCoreAnim(self=@n54)) -> (
    Melee:Exec
    Pistol:Exec
    Shotgun:Exec
    PaperBlueprint:Exec
    Rifle:Exec
    MeleeOneHand:Exec
    MachinePistol:Exec
    RocketLauncher:Exec
    GrenadeLauncher:Exec
    GoingCommando:Exec
    AssaultRifle:Exec
    TacticalShotgun:Exec
    SniperRifle:Exec
    TrapPlacement:Exec
    ShoulderLauncher:Exec
    AbilityDecoTool:Exec
    Crossbow:Exec
    C4:Exec
    RemoteControl:Exec
    DualWield:Exec
    AR_BullPup:Exec
    AR_ForwardGrip:Exec
    MedPackPaddles:Exec
    SMG_P90:Exec
    AR_DrumGun:Exec
    Consumable_Small:Exec
    Consumable_Large:Exec
    Balloon:Exec
    MountedTurret:Exec
    CreativeTool:Exec
    ExplosiveBow:Exec
    AshtonIndigo:Exec
    AshtonChicago:Exec
    MeleeDualWield:Exec
    Unarmed:Exec
   )
   n56 = Branch(execute=@n55.AssaultRifle, Condition=Get bIsTargeting())
   > Set AlphaAssaultRiffle(AlphaAssaultRiffle:float=1) -> (Output_Get:float)
   Set AlphaAssaultRiffle(execute=@n56.else, AlphaAssaultRiffle=0)
   n57 = OR Boolean!(A=Get bIsMoving2D(), B=Get bIsJumping(), C:bool=Get bIsFalling(), D:bool=Get bIsTargeting()) -> (ReturnValue)
   n58 = Branch(execute=@n55.Pistol, Condition=AND Boolean(A=Get bIsRelaxedLevel2() -> (bIsRelaxedLevel2:bool), B=NOT Boolean(A=@n57)))
   > Set AlphaHandRelaxed2(AlphaHandRelaxed2:float=1) -> (Output_Get:float)
   Set AlphaHandRelaxed2(execute=@n58.else, AlphaHandRelaxed2=0)
   n59 = Branch(execute=@n55.Shotgun, Condition=Get bIsTargeting())
   > Set AlphaShotgun(AlphaShotgun:float=1) -> (Output_Get:float)
   Set AlphaShotgun(execute=@n59.else, AlphaShotgun=0)
   n60 = Gravity Override State Function DEPRECATED(
    execute=@n7.then_0
    CharacterPartAnimInstance=Self-Reference()
    GameTimeInSeconds=GetGameTimeInSeconds()
    GravityOverrideParams=Get Gravity Override Params()
   )
   > Set GravityOverride(GravityOverride=@n60)
   > Is Valid(InputObject=Get PlayerPawn() -> (PlayerPawn:Fort Player Pawn Athena Object Reference))
   > n61 = Branch(Condition=Get bIsInFrontEnd())
   Set WPO Globals Function(
    execute=@n61.else
    Material=Get Mat Instance Dynamic()
    Overlay Material=Get Overlay Mat Instance Dynamic()
    Skeletal Mesh=Get SkelMesh()
    "Local Noise Transform  Location"=Get Local Noise Transform Location() -> (Local Noise Transform Location:Vector)
    Local Noise Offset=Get Local Noise Offset()
   )
   > Set WPO Regional Function!(
    execute
    Material:Material Instance Dynamic Object Reference=Get Mat Instance Dynamic()
    Overlay Material:Material Instance Dynamic Object Reference=Get Overlay Mat Instance Dynamic()
    Skeletal Mesh:Skeletal Mesh Component Object Reference=Get SkelMesh()
    Velocity Transform:Name=spine_03
    Velocity Interp Speed:double=Get Velocity Interp Speed()
    Collision Transform:Name=root
    Collision Transform Axis:EAxis Enum
    Invert Collision Transform Axis:bool=false
    Delta Time X:double=@n6.DeltaTimeX
    Previous Velocity Transform Location:"Vector (by ref)"=Get R1 Velocity Transform Location()
    Previous Velocity Transform Velocity:"Vector (by ref)"=Get R1 Velocity Transform Velocity()
    Region:int=1
   ) -> (then)
   > Set WPO Regional Function(
    Material=Get Mat Instance Dynamic()
    Overlay Material=Get Overlay Mat Instance Dynamic()
    Skeletal Mesh=Get SkelMesh()
    Velocity Transform=spine_01
    Velocity Interp Speed=Get Velocity Interp Speed()
    Collision Transform=root
    Invert Collision Transform Axis=true
    Delta Time X=@n6.DeltaTimeX
    Previous Velocity Transform Location=Get R2 Velocity Transform Location()
    Previous Velocity Transform Velocity=Get R2 Velocity Transform Velocity()
    Region=2
   )
   > n62 = Branch(Condition=OR Boolean!(A=Get bIsSkydiveFloating(), B=Get bIsParachuteOpened()) -> (ReturnValue))
   > Set WPO Pho Wind Intensity(WPO Pho Wind Intensity=0.8)
   > n63 = Set WPO Pho Turbulence(WPO Pho Turbulence:double=1) -> (Output_Get:double)
   SetScalarParameterValue(
    execute=[@n65.then, @n63.then]
    self=[Get Mat Instance Dynamic(), Get Overlay Mat Instance Dynamic()]
    ParameterName=ClothSim_WindStrength
    Value=Get WPO Pho Wind Intensity()
   )
   n64 = Get WPO Pho Turbulence() -> (WPO Pho Turbulence:double)
   > SetScalarParameterValue(self=[Get Mat Instance Dynamic(), Get Overlay Mat Instance Dynamic()], ParameterName=ClothSim_R1_TurbulenceStrength, Value=@n64)
   > SetScalarParameterValue(self=[Get Overlay Mat Instance Dynamic(), Get Mat Instance Dynamic()], ParameterName=ClothSim_R2_TurbulenceStrength, Value=@n64)
   Set WPO Pho Wind Intensity(execute=@n62.else, WPO Pho Wind Intensity=0)
   > n65 = Set WPO Pho Turbulence(WPO Pho Turbulence=0.1)
   Event OnCharacterCustomizationLoaded()
   > n66 = Set SkelMesh(SkelMesh=GetPartSkeletalMeshComponent())
   > n67 = Get Dynamic Material Instance(SkeletalMeshComponent=@n66)
   > Set Mat Instance Dynamic(Mat Instance Dynamic=@n67)
   > n68 = GetOrCreateDynamicOverlayMaterial(MeshComponent=@n66, MaterialName=None)
   > Set Overlay Mat Instance Dynamic(Overlay Mat Instance Dynamic=@n68)
  ]}
]
variables [
 Alpha Hand Relaxed 2 {type=number}
 AlphaAssaultRiffle {type=number}
 AlphaGun {type=boolean}
 AlphaHandRelaxed2 {type=number, default=1}
 AlphaShotgun {type=number}
 BeltAlpha {type=number}
 BeltUpAlpha {type=number}
 bkSkirtAlpha {type=number}
 Body_Mesh {type=SkeletalMeshComponent}
 CapeFixer {type=number}
 CrouchingAlpha {type=number}
 CrouchLevel_L {type=boolean}
 CrouchLevel_R {type=boolean}
 DBNOAlpha {type=number}
 EmoteDance {type=number}
 FrontEndLevel {type=number}
 FrSkirtAlpha {type=number}
 Gravity Override Params {type=GravityOverrideParamsStruct, default {skydive_GoingUp {gravityOverride=<0 200 -2000>}}}
 GravityOverride {type=Vector}
 Is Sprinting {type=number, default=1}
 isCrouchMove {type=number}
 isinAir {type=number}
 isInSkydive {type=number}
 isKneeSlide {type=number}
 isParachuting {type=number}
 isWalking {type=number}
 JacketAlpha {type=number}
 JumpAlpha {type=number}
 JumpMoveAlpha {type=number}
 Local Noise Offset {type=Vector}
 Local Noise Transform Location {type=Vector}
 Mat Instance Dynamic {type=MaterialInstanceDynamic}
 OnGrindRail {type=number}
 OnZipline {type=boolean}
 Overlay Mat Instance Dynamic {type=MaterialInstanceDynamic}
 PlayerPawn {type=FortPlayerPawnAthena}
 R1 Velocity Transform Location {type=Vector}
 R1 Velocity Transform Velocity {type=Vector}
 R2 Velocity Transform Location {type=Vector}
 R2 Velocity Transform Velocity {type=Vector}
 RotationCrouch_R {type=Rotator, default=<rot 7.5 7.5 0>}
 RotationCrouched_L {type=Rotator, default=<rot 15.192 0.001 -0.007>}
 Sim Space Settings Regular {
  type=SimSpaceSettings
  default {
   dampingAlpha=1
   maxAngularAcceleration=10000
   maxAngularVelocity=10000
   maxLinearAcceleration=10000
   maxLinearVelocity=10000
   velocityScaleZ=1
   worldAlpha=0
   externalLinearDragV=<0 0 0>
  }
 }
 Sim Space Settings SkyDive {
  type=SimSpaceSettings
  default {
   dampingAlpha=1
   maxAngularAcceleration=10000
   maxAngularVelocity=10000
   maxLinearAcceleration=10000
   maxLinearVelocity=10000
   velocityScaleZ=1
   worldAlpha=0
   externalLinearDragV=<0 0 0>
  }
 }
 Sim Space Settings Skydiving {
  type=SimSpaceSettings
  default {
   dampingAlpha=1
   maxAngularAcceleration=10000
   maxAngularVelocity=10000
   maxLinearAcceleration=10000
   maxLinearVelocity=10000
   velocityScaleZ=1
   worldAlpha=0
   externalLinearDragV=<0 0 0>
  }
 }
 SkelMesh {type=SkeletalMeshComponent}
 SkirtDynAlpha {type=number}
 skydiveAlpha {type=number}
 SprintAlpha {type=number}
 Standing {type=number}
 thigh_rot_l {type=number}
 thigh_rot_r {type=number}
 TranslationCrouch_R {type=Vector, default=<-5 10 -7.25>}
 TranslationCrouched_L {type=Vector, default=<-4.393 11.791 12.406>}
 Velocity Interp Speed {type=number}
 WPO Pho Turbulence {type=number}
 WPO Pho Wind Intensity {type=number}
]
